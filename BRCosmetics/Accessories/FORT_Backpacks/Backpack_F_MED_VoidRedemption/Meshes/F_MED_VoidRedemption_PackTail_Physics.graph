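# Physics asset for the VoidRedemption backpack cosmetic's PackTail mesh, defining collision and joints for its tail and two legs.
# A capsule and box on body_c form the central body, with single capsules on each leg and five tail bones.
# The tail chains from body_c through tight swing limits and zero twist limits.
# Both legs attach directly to body_c with free swing and twist, despite having limit angles set.
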PhysicsAsset /BRCosmetics/Accessories/FORT_Backpacks/Backpack_F_MED_VoidRedemption/Meshes/F_MED_VoidRedemption_PackTail_Physics
bodies {
 body_c [
  {shapeType=Capsule, length=7.2545, radius=10.1447, shapeName=body_c_capsule, center=<0 -1.4195 0>, rotation=<rot -90 -90 90.7258>}
  {
   shapeType=Box
   extentX=4.1555
   extentY=19.6722
   extentZ=25.9352
   shapeName=body_c_box
   center=<-9.2348 -4.8565 0>
   rotation=<rot 0 -8.6475 0>
  }
 ]
 dyn_leg_l [{
   shapeType=Capsule
   length=2.2379
   radius=2.2572
   shapeName=dyn_leg_l_capsule
   center=<3.1617 0 0.0757>
   rotation=<rot -85.0002 -0.0004 0.0004>
  }]
 dyn_leg_r [{
   shapeType=Capsule
   length=2.2379
   radius=2.2572
   shapeName=dyn_leg_l_capsule
   center=<3.074 0 0.1563>
   rotation=<rot -85.0002 -0.0004 -179.9996>
  }]
 dyn_tail_01 [{shapeType=Capsule, length=1.653, radius=1.255, shapeName=dyn_tail_01_capsule, center=<1.0665 0 0>, rotation=<rot -90 -179.0505 179.0507>}]
 dyn_tail_02 [{
   shapeType=Capsule
   length=2.1583
   radius=1.306
   shapeName=dyn_tail_02_capsule
   center=<0.3149 0 -0.0555>
   rotation=<rot -79.9999 -179.9996 179.9997>
  }]
 dyn_tail_03 [{shapeType=Capsule, length=1.6881, radius=1.3118, shapeName=dyn_tail_03_capsule, rotation=<rot 80 0.0003 0>}]
 dyn_tail_04 [{
   shapeType=Capsule
   length=2.465
   radius=1.2825
   shapeName=dyn_tail_04_capsule
   center=<0.1254 0.0229 -0.0336>
   rotation=<rot 75 -0.0004 -10.0005>
  }]
 dyn_tail_05 [{
   shapeType=Capsule
   length=2.3849
   radius=1.2273
   shapeName=dyn_tail_05_capsule
   center=<1.542 0 -0.1005>
   rotation=<rot 84.7445 178.3726 -178.0321>
  }]
}
constraints [
 {bone1Name=dyn_tail_02, bone2Name=dyn_tail_01, swing1LimitDegrees=6, swing2LimitDegrees=5, twistLimitDegrees=0}
 {bone1Name=dyn_tail_03, bone2Name=dyn_tail_02, swing1LimitDegrees=4, swing2LimitDegrees=4, twistLimitDegrees=0}
 {bone1Name=dyn_tail_04, bone2Name=dyn_tail_03, swing1LimitDegrees=3, swing2LimitDegrees=4, twistLimitDegrees=0}
 {bone1Name=dyn_tail_05, bone2Name=dyn_tail_04, swing1LimitDegrees=3, swing2LimitDegrees=3, twistLimitDegrees=0}
 {
  bone1Name=dyn_leg_l
  bone2Name=body_c
  swing1LimitDegrees=14
  swing1Motion=Free
  swing2LimitDegrees=14
  swing2Motion=Free
  twistLimitDegrees=5
  twistMotion=Free
 }
 {bone1Name=dyn_tail_01, bone2Name=body_c, swing1LimitDegrees=7, swing2LimitDegrees=6, twistLimitDegrees=0}
 {
  bone1Name=dyn_leg_r
  bone2Name=body_c
  swing1LimitDegrees=17
  swing1Motion=Free
  swing2LimitDegrees=17
  swing2Motion=Free
  twistLimitDegrees=5
  twistMotion=Free
 }
]
