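# Physics asset for the Void Redemption backpack cosmetic, defining ragdoll-style secondary motion for a figure with arms, legs, ears, earrings, a chain, a head loop and a zipper mouth.
# Every bone carries a single capsule, and most dynamic parts constrain directly to the central body capsule.
# The chain links hang in series from the body, and one earring hangs from the right ear.
# Narrow swing limits and little or no twist keep motion subtle.
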PhysicsAsset /BRCosmetics/Accessories/FORT_Backpacks/Backpack_F_MED_VoidRedemption/Meshes/F_MED_VoidRedemption_Pack_Physics
bodies {
 body_c [{shapeType=Capsule, length=7.2545, radius=10.1447, shapeName=body_c_capsule, center=<0 -1.4195 0>, rotation=<rot -90 -90 90.7258>}]
 dyn_arm_l [{shapeType=Capsule, length=2.4294, radius=2.5726, center=<7.2426 0 -0.1295>, rotation=<rot 90 178.3124 178.3123>}]
 dyn_arm_r [{shapeType=Capsule, length=2.4193, radius=2.4972, center=<6.9222 0 0>, rotation=<rot -84.2026 -179.9999 -180>}]
 dyn_chain_01 [{shapeType=Capsule, length=0, radius=0.7086}]
 dyn_chain_02 [{shapeType=Capsule, length=0.2989, radius=0.7011, center=<0.581 0 0>, rotation=<rot -90 -10.0251 10.0254>}]
 dyn_chain_03 [{shapeType=Capsule, length=0.1896, radius=1.6976, center=<2.5873 0.0374 0>}]
 dyn_ear_l [{shapeType=Capsule, length=1.3179, radius=2.3317, shapeName=dyn_ear_l_capsule, center=<2.4145 0 0>, rotation=<rot 90 0 0.0002>}]
 dyn_ear_r [{shapeType=Capsule, length=1.3179, radius=2.3317, shapeName=dyn_ear_l_capsule, center=<2.7278 0 0>, rotation=<rot 90 0 -179.9998>}]
 dyn_earring_ear [{
   shapeType=Capsule
   length=1.5116
   radius=1.5112
   shapeName=dyn_earring_ear_capsule
   center=<1.0994 -0.0101 -0.0278>
   rotation=<rot 0.9715 -90 -71.4813>
  }]
 dyn_earring_head [{shapeType=Capsule, length=0, radius=1.1383, shapeName=dyn_earring_head_capsule, center=<0.5771 0 0>}]
 dyn_leg_l [{
   shapeType=Capsule
   length=2.2379
   radius=2.2572
   shapeName=dyn_leg_l_capsule
   center=<3.1617 0 0.0757>
   rotation=<rot -85.0002 -0.0004 0.0004>
  }]
 dyn_leg_r [{
   shapeType=Capsule
   length=2.2379
   radius=2.2572
   shapeName=dyn_leg_l_capsule
   center=<3.074 0 0.1563>
   rotation=<rot -85.0002 -0.0004 -179.9996>
  }]
 dyn_loopHead [{shapeType=Capsule, length=4.09, radius=1.5422, shapeName=dyn_loopHead_capsule, center=<2.5471 0 0>, rotation=<rot 0.0001 0 -85>}]
 dyn_zipperMouthw [{
   shapeType=Capsule
   length=2.7654
   radius=0.9648
   shapeName=dyn_zipperMouthw_capsule
   center=<2.9597 0 0.2166>
   rotation=<rot -90 -90 89.9996>
  }]
}
constraints [
 {bone1Name=dyn_chain_02, bone2Name=dyn_chain_01, swing1LimitDegrees=2, swing2LimitDegrees=16, twistLimitDegrees=0}
 {bone1Name=dyn_chain_03, bone2Name=dyn_chain_02, swing1LimitDegrees=25, swing2LimitDegrees=2, twistLimitDegrees=0}
 {bone1Name=dyn_leg_l, bone2Name=body_c, swing1LimitDegrees=14, swing2LimitDegrees=14, twistLimitDegrees=5}
 {bone1Name=dyn_ear_l, bone2Name=body_c, swing1LimitDegrees=12, swing2LimitDegrees=12, twistLimitDegrees=3}
 {bone1Name=dyn_loopHead, bone2Name=body_c, swing1LimitDegrees=10, swing2LimitDegrees=10, twistLimitDegrees=0}
 {bone1Name=dyn_chain_01, bone2Name=body_c, swing1LimitDegrees=2, swing2LimitDegrees=16, twistLimitDegrees=0}
 {bone1Name=dyn_ear_r, bone2Name=body_c, swing1LimitDegrees=15, swing2LimitDegrees=15, twistLimitDegrees=3}
 {bone1Name=dyn_zipperMouthw, bone2Name=body_c, swing1LimitDegrees=18, swing2LimitDegrees=18, twistLimitDegrees=0}
 {bone1Name=dyn_earring_ear, bone2Name=dyn_ear_r, swing1LimitDegrees=15, swing2LimitDegrees=0, twistLimitDegrees=0}
 {bone1Name=dyn_leg_r, bone2Name=body_c, swing1LimitDegrees=17, swing2LimitDegrees=17, twistLimitDegrees=5}
 {bone1Name=dyn_earring_head, bone2Name=body_c, swing1LimitDegrees=20, swing2LimitDegrees=4, twistLimitDegrees=0}
 {bone1Name=dyn_arm_r, bone2Name=body_c, swing1LimitDegrees=12, swing2LimitDegrees=10, twistLimitDegrees=3}
 {bone1Name=dyn_arm_l, bone2Name=body_c, swing1LimitDegrees=20, swing2LimitDegrees=14, twistLimitDegrees=4}
]
